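# HUD tag prompt container widget for the Jethro DisguiseBooth feature, derived from FortHUDTagPromptContainerWidget.
# A SizeBox with minimum desired dimensions wraps the inherited OverlayWidget, alongside an unparented TutorialHighlight entry.
# Intro and Outro animations fade the overlay's render opacity in and out.
# The event graph is empty, so the Blueprint supplies only layout and animations, with no logic of its own.
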WidgetBlueprint /JethroLager/DisguiseBooth/UI/Jethro_HUDTagPromptWidget
animations [
 Intro {
  binding_count=1
  length_seconds=1.2143
  object=&self:Intro
  sequence {bindings [OverlayWidget {
     id=#1
     possessed_class=Overlay
     tracks [{
       type=MovieSceneFloatTrack
       display_name=Render Opacity
       sections [{type=MovieSceneFloatSection, channels [None {type=MovieSceneScriptingFloatChannel, keys [[6.8173,0],[31.104,1]]}]}]
      }]
    }]}
 }
 Outro {
  binding_count=1
  length_seconds=0.7391
  object=&self:Outro
  sequence {bindings [OverlayWidget {
     id=#2
     possessed_class=Overlay
     tracks [{
       type=MovieSceneFloatTrack
       display_name=Render Opacity
       sections [{type=MovieSceneFloatSection, end_frame=14, start_frame=0, channels [None {type=MovieSceneScriptingFloatChannel, keys [[0,1],[14.7813,0]]}]}]
      }]
    }]}
 }
]
graphs [EventGraph {links [], nodes []}]
variables []
widgets {
 info {inheritedWidgetCount=2, parentClass=/Script/FortniteUI.FortHUDTagPromptContainerWidget, rootWidgetClass=&SizeBox, widgetCount=3}
 widgets [
  SizeBox_0 {
   widgetClassPath=&SizeBox
   widget=SizeBox_0
   properties {bOverride_MinDesiredHeight=true, bOverride_MinDesiredWidth=true, minDesiredHeight=100, minDesiredWidth=600}
  }
  OverlayWidget {widgetClassPath=&Overlay, bInherited=true, parent=SizeBox_0, widget=OverlayWidget, slot {class=SizeBoxSlot}}
  TutorialHighlight {widgetClassPath=&TutorialHighlight}
 ]
}
